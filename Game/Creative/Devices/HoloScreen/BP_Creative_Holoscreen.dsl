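; Creative HoloScreen device that projects a selectable holographic screen with configurable size, bend, rotation and projector visibility.
; Toy option load or change re-initializes it, and cosmetic-capable instances look up the HoloScreens entry by index.
; That entry's material and texture async-load separately, with completion flags gating the texture update.
; Dynamic screen and laser materials receive color, glow, contrast and motion parameters, and override the projector mesh's materials.
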(event EventBeginPlay
  (Visibility|SetVisibilityComponents (Variables|Default|GetCreative_VisibleInGame_Component) (Utilities|Array|MakeArray (Variables|Debug|GetStaticMeshComponent)) true)
  (if (Variables|HoloScreen|GetInitialize%20without%20Toy%20Options)
    (CallFunction|InitializeDevice)))

(event Custom|LoadHoloscreen
  (bind _holoscreenindex (Variables|HoloScreen|GetHoloScreenIndex))
  (bind _holoscreens (Variables|Default|GetHoloScreens))
  (bind _output (Utilities|Array|Get(aref) _holoscreens _holoscreenindex))
  (bind (_holoscreenmaterial_30_ed8c74eb462773d4e60fa887aa380fd1 _brotates_48_0c2c4f8d4f6cd3d1f43cf68c170a596e _color_11_9c4e1fe245b765a4e5e77cbb4f4a8f29 _screenglow_42_e9a50eb14063f7ab7f48f58c76539ead _laserglow_43_e93a0ee8478f862546d3798e61d5f7af _screencontrast_19_86992e794e3316506781a5aa34d073a0 _texture_22_fb56d2e3416625bfa30dad83eadf0552) (Utilities|Struct|BreakHoloscreenDataStruct _output))
  (if (Utilities|Array|IsValidIndex (Variables|Default|GetHoloScreens) (Variables|HoloScreen|GetHoloScreenIndex))
    (Variables|Default|SetCompletedTextureLoad false)
    (Variables|Default|SetCompletedMaterialLoad false)
    (Utilities|FlowControl|Sequence
      (:then_0
        (bind _loaded_object (Creative|Loading|AsyncLoadEditorSafe _holoscreenmaterial_30_ed8c74eb462773d4e60fa887aa380fd1)
          (:Loaded
            (Utilities|ValidateAsyncLoadedObject _holoscreenmaterial_30_ed8c74eb462773d4e60fa887aa380fd1 _loaded_object
              (:True
                (bind _asmaterial_instance (Utilities|Casting|CastToMaterialInstance _loaded_object))
                (bind _output_get (Variables|Default|SetCurrentMaterial _asmaterial_instance))
                (Rendering|Material|SetMaterial (Variables|Default|GetSM_HoloProjector) 0 _output_get) ; Put material on projector screen
                (Variables|Default|SetCurrentScreenMID (Material|GetMID (Variables|Default|GetSM_HoloProjector))) ; Make material dynamic
                (Material|ReattachOrCreateMID (Variables|Default|GetSM_HoloProjector) (Variables|Default|GetCurrentScreenMID)) ; Assign dynamic material to projector screen
                (Variables|Default|SetCompletedMaterialLoad true)
                (CallFunction|SetHoloScreenScale)
                (CallFunction|SetMIDVariables (select _brotates_48_0c2c4f8d4f6cd3d1f43cf68c170a596e (Variables|Default|GetRotationSpeed) 0.0) (Variables|HoloScreen|GetBend) _color_11_9c4e1fe245b765a4e5e77cbb4f4a8f29 _screenglow_42_e9a50eb14063f7ab7f48f58c76539ead _laserglow_43_e93a0ee8478f862546d3798e61d5f7af _screencontrast_19_86992e794e3316506781a5aa34d073a0))))))
      (:then_1
        (bind _loaded_object_2 (Creative|Loading|AsyncLoadEditorSafe _texture_22_fb56d2e3416625bfa30dad83eadf0552)
          (:Loaded
            (Utilities|ValidateAsyncLoadedObject _texture_22_fb56d2e3416625bfa30dad83eadf0552 _loaded_object_2
              (:True
                (bind _astexture (Utilities|Casting|CastToTexture _loaded_object_2))
                (Variables|Default|SetCurrentTexture _astexture)
                (Variables|Default|SetCompletedTextureLoad true)))))))
    (if (and (Variables|Default|GetCompletedTextureLoad) (Variables|Default|GetCompletedMaterialLoad))
      (CallFunction|UpdateMIDTexture))))

(event OnAnyPropertyChanged(ToyOptionsComponent)
  ;[0 1 0.04321 0.5] Initialize Device
  (CallFunction|InitializeDevice))

(event OnPropertiesLoaded(ToyOptionsComponent)
  ;[0 1 0.04321 0.5] Initialize Device
  (CallFunction|InitializeDevice))

(event Custom|Initialize%20Device
  ;[0 1 0.04321 0.5] Initialize Device
  (Visibility|SetVisiblityDuringPhase (Variables|Default|GetCreative_VisibleInGame_Component) (select (Variables|HoloScreen|GetbProjectorVisible) "NewEnumerator3" "NewEnumerator1"))
  (Networking|CanExecuteCosmeticEvents
    (:True
      (CallFunction|LoadHoloscreen)
      (CallFunction|SetHoloScreenScale))))

(fn GetOverrideMeshMaterial
  (params (MeshComp :type Static%20Mesh%20Component%20Object%20Reference) (MatIdx :type Integer))
  (if (Utilities|Equal(Object) MeshComp (Variables|Default|GetSM_HoloProjector))
    (return (Utilities|Select :Index MatIdx :Option%200 (Variables|Default|GetCurrentScreenMID) :Option%201 (Variables|Default|GetCurrentLaserMID)))
    (else
      (return (|CallParentFunction MeshComp MatIdx)))))

(fn IsTargetableByLinearMover
  (return true))

(fn SetMIDVariables
  (params (RotationSpeed :type Float%20%28double-precision%29) (BendAmount :type Float%20%28double-precision%29) (Color :type Linear%20Color%20Structure) (ScreenGlow :type Float%20%28double-precision%29) (LaserGlow :type Float%20%28double-precision%29) (ScreenContrast :type Float%20%28double-precision%29))
  (bind _currentscreenmid (Variables|Default|GetCurrentScreenMID))
  (bind _sm_holoprojector (Variables|Default|GetSM_HoloProjector))
  (Material|ReattachOrCreateMID _sm_holoprojector (Variables|Default|GetCurrentLaserMID) 1)
  (Material|ReattachOrCreateMID _sm_holoprojector (Variables|Default|GetCurrentScreenMID))
  (CallFunction|SetRotationSpeed (Variables|Default|GetRotationSpeed))
  (CallFunction|SetBend (Variables|GetBendAmount))
  (Rendering|Material|SetVectorParameterValue _currentscreenmid "Color" (Variables|GetColor) :!class /Script/Engine.MaterialInstanceDynamic)
  (Rendering|Material|SetScalarParameterValue _currentscreenmid "Glow" (Variables|GetScreenGlow) :!class /Script/Engine.MaterialInstanceDynamic)
  (Rendering|Material|SetScalarParameterValue (Variables|Default|GetCurrentLaserMID) "Glow" (Variables|GetLaserGlow) :!class /Script/Engine.MaterialInstanceDynamic)
  (Rendering|Material|SetScalarParameterValue _currentscreenmid "Contrast" (Variables|GetScreenContrast) :!class /Script/Engine.MaterialInstanceDynamic))

(fn SetBend
  (params (Bend :type Float%20%28double-precision%29))
  (if (and (Utilities|IsValid (Variables|Default|GetCurrentScreenMID) :!class /Script/Engine.KismetSystemLibrary) (Utilities|IsValid (Variables|Default|GetCurrentLaserMID) :!class /Script/Engine.KismetSystemLibrary))
    (Rendering|Material|SetScalarParameterValue (Variables|Default|GetCurrentScreenMID) "Rotationamount" (* Bend -0.01) :!class /Script/Engine.MaterialInstanceDynamic)))

(fn SetHoloScreenScale
  (bind _height (Variables|HoloScreen|GetHeight))
  (bind _distance (Variables|HoloScreen|GetDistance))
  (bind _width (Variables|HoloScreen|GetWidth))
  (bind _returnvalue (Math|Vector|MakeVector _width _distance _height))
  (Transformation|SetWorldScale3D (Variables|Default|GetSM_HoloProjector) _returnvalue)
  (Transformation|SetWorldScale3D (Variables|Default|GetBox) _returnvalue))

(fn SetRotationSpeed
  (params (InputPin :type Float%20%28double-precision%29))
  (Rendering|Material|SetScalarParameterValue (Variables|Default|GetCurrentScreenMID) "Constant Rotation Speed" InputPin :!class /Script/Engine.MaterialInstanceDynamic))

(fn ShouldInvisiblePropsBeAllowed
  (return false))

(fn UpdateMIDTexture
  (if (Math|Boolean|ANDBoolean (Utilities|IsValid (Variables|Default|GetCurrentTexture) :!class /Script/Engine.KismetSystemLibrary) (Utilities|IsValid (Variables|Default|GetCurrentScreenMID) :!class /Script/Engine.KismetSystemLibrary) (Utilities|IsValid (Variables|Default|GetCurrentLaserMID) :!class /Script/Engine.KismetSystemLibrary))
    (Rendering|Material|SetTextureParameterValue (Variables|Default|GetCurrentScreenMID) "BaseColor" (Variables|Default|GetCurrentTexture))))

(fn ConstructionScript)
